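; Rift Gate portal ring prop that shows an overloaded state driven by calendar events.
; On authority, late RG01 or active RG02 sets OverloadedState, whose OnRep toggles material wobble and camera shake and starts looping force feedback.
; A delayed bind to BP_Asteria_Rift_Setup feeds the ambient audio parameter, unbound on death as audio fades.
; Box overlap triggers a player lens cue; debug event SetOLS writes the Overload collection parameter and state.
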(event Custom|SetOLS
  (params (OverloadAmount :type Float%20%28double-precision%29))
  ;[1 1 1] Set Overload Amount Debug
  (Rendering|Material|SetScalarParameterValue "/Game/Athena/Items/ForagedItems/Rift/FX/RiftGate/MPC_Riftgate.MPC_Riftgate" "Overload" OverloadAmount :!class /Script/Engine.KismetMaterialLibrary)
  (if (>= OverloadAmount 0.8)
    (Variables|Default|SetOverloadedState true)
    (else
      (Variables|Default|SetOverloadedState false))))

(event EventBeginPlay
  (Utilities|FlowControl|Sequence
    (:then_0
      ;[1 1 1] Setting Overloaded state effects in this BP ( WPO MID and Camera Shake ON)
      (Networking|SwitchHasAuthority
        (:Authority
          (Class|FortAsyncActionCurrentPlaylistReady|PlaylistDataReadyAsync)
          (bind (_biseventactive _timesincebegin _timeuntilend _timespanratio) (CalendarEvents|GetCalendarEventInformation "RG01"))
          (if _biseventactive
            (if (>= _timespanratio 0.8)
              _)
            (else
              (bind _biseventactive_2 (CalendarEvents|GetCalendarEventInformation "RG02"))
              (if _biseventactive_2
                _)))
          (Variables|Default|SetOverloadedState true))))
    (:then_1
      (Utilities|FlowControl|Delay 2.0)
      ;[1 1 1] Getting audio parameter from Rift Setup
      (bind _returnvalue (Actor|GetActorOfClass "/Game/Athena/Asteria/Blueprints/BP_Asteria_Rift_Setup.BP_Asteria_Rift_Setup_C"))
      (Utilities|IsValid _returnvalue
        (:Is%20Valid
          (bind _output_get (Variables|Default|SetRiftSetUpRef _returnvalue))
          (Default|BindEventtoOverloadAudioParamED _output_get (delegate Custom|SetPortalAudioParameter)))))))

(event Custom|SetPortalAudioParameter
  (params (AudioParam :type Float%20%28double-precision%29))
  ;[1 1 1] Getting audio parameter from Rift Setup
  (Audio|Parameter|SetFloatParameter (Variables|Default|GetAmbient%20Audio) "CalendarControl" AudioParam :!class /Script/Engine.AudioComponent))

(event EventOnDeathPlayEffects
  (params (Damage :type Float%20%28single-precision%29) (DamageTags :type Gameplay%20Tag%20Container%20Structure%20%28by%20ref%29) (Momentum :type Vector :default "0, 0, 0") (HitInfo :type Hit%20Result%20Structure%20%28by%20ref%29) (InstigatedBy :type Fort%20Pawn%20Object%20Reference) (DamageCauser :type Actor%20Object%20Reference) (EffectContext :type Gameplay%20Effect%20Context%20Handle%20Structure))
  (Audio|Components|Audio|FadeOut (Variables|Default|GetAmbient%20Audio) 1.0 0.0 "Linear" :!class /Script/Engine.AudioComponent)
  (Audio|SpawnSoundatLocation "/Game/Athena/Sounds/Props/RiftGate/Rift_Gate_Device_Destroyed_Amb_Loop_Stop_Cue.Rift_Gate_Device_Destroyed_Amb_Loop_Stop_Cue" (Transformation|GetWorldLocation (Variables|Default|GetAmbient%20Audio)))
  (bind _riftsetupref (Variables|Default|GetRiftSetUpRef)
    (:then
      (Default|UnbindallEventsfromOverloadAudioParamED _riftsetupref))
    (:else)))

(event OnComponentBeginOverlap(Box)
  (params (OverlappedComponent :type Primitive%20Component%20Object%20Reference) (OtherActor :type Actor%20Object%20Reference) (OtherComp :type Primitive%20Component%20Object%20Reference) (OtherBodyIndex :type Integer) (bFromSweep :type Boolean) (SweepResult :type Hit%20Result%20Structure%20%28by%20ref%29))
  (Utilities|IsValid OtherActor
    (:Is%20Valid
      (bind _asfort_player_pawn (Utilities|Casting|CastToFortPlayerPawn OtherActor))
      (GameplayCue|ExecuteGameplayCue _asfort_player_pawn (Variables|Default|GetGC_LensEffect)))))

(fn OnRep_OverloadedState
  (bind _sm_riftgate_portal_a (Variables|Default|GetSM_RiftGate_Portal_A))
  (bind _sm_riftgate_portal_a_2 (Variables|Default|GetSM_RiftGate_Portal_A))
  (if (Variables|Default|GetOverloadedState)
    (Rendering|Material|SetScalarParameterValueonMaterials _sm_riftgate_portal_a "WPO_WindNew__X Val" 3.0)
    (Rendering|Material|SetScalarParameterValueonMaterials _sm_riftgate_portal_a "WPO_WindNew_Y Val" 3.0)
    (CameraShake|Start (Variables|Default|GetCameraShakeSource))
    (ForceFeedback|SpawnForceFeedbackatLocation :ForceFeedbackEffect "/Game/Environments/Asteria/Props/LightCastle/LightCastle_RiftGate_A/RiftGate_Portal_A/ForceFeedback/FF_RiftGate_Portal_B.FF_RiftGate_Portal_B" :Location (Transformation|GetWorldLocation (Variables|Default|GetForcefeedback)) :bLooping true :IntensityMultiplier 1.0 :StartTime 0.0 :AttenuationSettings "/Game/Environments/Asteria/Props/LightCastle/LightCastle_RiftGate_A/RiftGate_Portal_A/ForceFeedback/FF_RiftGate_Portal_B_Attenuation.FF_RiftGate_Portal_B_Attenuation" :bAutoDestroy true)
    (else
      (Rendering|Material|SetScalarParameterValueonMaterials _sm_riftgate_portal_a_2 "WPO_WindNew__X Val" 0.0)
      (Rendering|Material|SetScalarParameterValueonMaterials _sm_riftgate_portal_a_2 "WPO_WindNew_Y Val" 0.0)
      (CameraShake|StopAllCameraShakes (Variables|Default|GetCameraShakeSource) true))))

(fn ConstructionScript
  ;[1 1 1] Creating MID and assigning it to Ring Pieces
  (bind _returnvalue (Rendering|Material|CreateDynamicMaterialInstance "/Game/Environments/Asteria/Props/LightCastle/LightCastle_RiftGate_A/RiftGate_Portal_A/Materials/MI_RiftGate_Portal_A.MI_RiftGate_Portal_A" "" "None" :!class /Script/Engine.KismetMaterialLibrary))
  (bind _output_get (Variables|Default|SetMID _returnvalue))
  (Rendering|Material|SetMaterial (Variables|Default|GetSM_RiftGate_Portal_A) 0 _output_get))
